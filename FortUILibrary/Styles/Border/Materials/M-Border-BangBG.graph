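# Translucent unlit UI material that draws a procedural "bang" border background shape from UV math, without textures.
# Centered UVs form a slanted, soft-edged band through absolute distance and SmoothStep, with parameters for rotation, offset and edge softness.
# UseSpecialBang blends that mask toward its product with a second band on the other axis.
# The same parameter blends the color from static RGB1 to a time-driven sine pulse between RGB1 and RGB2.
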Material /FortUILibrary/Styles/Border/Materials/M-Border-BangBG
graph {
 properties {
  bUsedWithStaticMesh=false
  blendMode=BLEND_Translucent
  materialDomain=MD_UI
  pixelDepthOffsetMode=PDOM_Legacy
  shadingModel=MSM_Unlit
  usedShadingModels=Unlit
 }
 nodes [
  n1 = TextureCoordinate()
  n2 = MaterialFunctionCall(Float2=Add(A=@n1) {constB=-0.5}) {materialFunction=&BreakOutFloat2Components}
  n3 = Add(A=Multiply(A=@n2.R, B=ScalarParameter() {defaultValue=-0.6835, parameterName=Rotation}), B=@n2.G)
  n4 = ScalarParameter() {defaultValue=0.18, parameterName=SS Min}
  n5 = ScalarParameter() {defaultValue=0.125, parameterName=SS Max}
  n6 = MaterialFunctionCall(Alpha=Subtract(A=Abs(None=@n3), B=ScalarParameter() {defaultValue=0.2326, parameterName=Offset}), Min=@n4, Max=@n5) {materialFunction=&SmoothStep}
  n7 = MaterialFunctionCall(Float2=Add(A=@n1) {constB=-0.5}) {materialFunction=&BreakOutFloat2Components}
  n8 = Multiply(A=@n7.G, B=ScalarParameter() {parameterName=Rotation_1})
  n9 = Subtract(A=Abs(None=Add(A=@n8, B=@n7.R)), B=ScalarParameter() {defaultValue=0.1, parameterName=Offset_1})
  n10 = ScalarParameter() {parameterName=UseSpecialBang}
  LinearInterpolate(
   A=@n6.Result
   B=DotProduct(
    A=MaterialFunctionCall(Alpha=@n9, Min=@n4, Max=@n5) {materialFunction=&SmoothStep}.Result
    B=@n6.Result
   )
   Alpha=@n10
  )
  n11 = VectorParameter() {parameterName=RGB1, defaultValue=<rgba 1 0.9065 0.0648 1>}
  n12 = Multiply(A=Time(), B=ScalarParameter() {defaultValue=0.25, parameterName=TimeMultiplier})
  n13 = Multiply(A=Add(A=Sine(None=@n12))) {constB=0.5}
  LinearInterpolate(
   A=@n11.RGB
   B=LinearInterpolate(A=@n11.RGB, B=VectorParameter() {parameterName=RGB2, defaultValue=<rgba 0 0 1 1>}.RGB, Alpha=Clamp(None=@n13))
   Alpha=@n10
  )
 ]
}
