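# UI material for a translucent difficulty-increase border background, layering blue gradients over a dark base color.
# Two exponential radial gradients form the color blend mask, lifted by a panning, rotated sine-line grid term.
# The grid's third sine subtracts a channel from itself, so its final blend saturates to a constant.
# The IsLockDown scalar blends in an orange Gradient-Col3 glow shaped by a rotated, powered radial gradient.
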Material /FortUILibrary/Styles/Border/Materials/M-Border-DiffIncrease-Background
graph {
 properties {
  bUsedWithStaticMesh=false
  blendMode=BLEND_Translucent
  materialDomain=MD_UI
  numCustomizedUVs=2
  pixelDepthOffsetMode=PDOM_Legacy
  usedShadingModels=Default Lit
 }
 nodes [
  Constant() {r=0.925}
  n1 = Multiply(A=TextureCoordinate(), B=Constant2Vector() {g=1, r=3})
  n2 = Constant2Vector() {r=-0.999}
  n3 = Constant2Vector() {g=0.95, r=0.5}
  n4 = Constant() {r=1.5}
  n5 = Constant() {r=3}
  n6 = Add(
   A=MaterialFunctionCall(UVs=Add(A=@n1, B=@n2), CenterPosition=Constant2Vector() {r=0.5}, Radius=Constant() {r=1.25}, Density=Constant() {r=0.75}) {materialFunction=&RadialGradientExponential}.RadialGradientExponential
   B=OneMinus(None=MaterialFunctionCall(CenterPosition=@n3, Radius=@n4, Density=@n5) {materialFunction=&RadialGradientExponential}.RadialGradientExponential)
  )
  n7 = Rotator(Coordinate=Panner(Coordinate=TextureCoordinate() {uTiling=10, vTiling=4}) {speedX=0.125}, Time=ScalarParameter() {defaultValue=0.1125}) {centerX=5, centerY=2, speed=6.2932}
  n8 = MaterialFunctionCall(Float2=@n7) {materialFunction=&BreakOutFloat2Components}
  n9 = Abs(None=Sine(None=@n8.R) {period=2})
  n10 = OneMinus(None=Multiply(A=Subtract(A=@n9) {constB=0.025}) {constB=32})
  n11 = Abs(None=Sine(None=@n8.G) {period=2})
  n12 = OneMinus(None=Multiply(A=Subtract(A=@n11) {constB=0.025}) {constB=32})
  n13 = MaterialFunctionCall(Float2=@n7) {materialFunction=&BreakOutFloat2Components}
  n14 = Sine(None=Subtract(A=@n13.R, B=@n13.R)) {period=2}
  n15 = Multiply(A=Subtract(A=Abs(None=@n14)) {constB=0.0375}) {constB=32}
  n16 = LinearInterpolate(A=LinearInterpolate(A=Clamp(None=@n10), Alpha=Clamp(None=@n12)), Alpha=Clamp(None=OneMinus(None=@n15)))
  n17 = Add(A=@n6, B=LinearInterpolate(B=Multiply(A=@n16) {constB=0.125}, Alpha=Power(Base=@n6) {constExponent=0.75}))
  n18 = LinearInterpolate(
   A=VectorParameter() {parameterName=Background-Col, defaultValue=<rgba 0.004 0.0052 0.0168 1>}.RGB
   B=LinearInterpolate(
    A=VectorParameter() {parameterName=Gradient-Col2, defaultValue=<rgba 0.0021 0.0999 0.8148 1>}.RGB
    B=VectorParameter() {parameterName=Gradient-Col1, defaultValue=<rgba 0.0887 0.3864 0.9301 1>}.RGB
    Alpha=@n17
   )
   Alpha=LinearInterpolate(Alpha=@n17) {constB=0.75}
  )
  n19 = TextureCoordinate()
  n20 = Constant() {r=0.025}
  n21 = MaterialFunctionCall(
   UVs=Multiply(A=Rotator(Coordinate=@n19, Time=@n20) {centerY=0.25, speed=6.2932}, B=Constant2Vector() {g=0.5, r=1})
   CenterPosition=ComponentMask(None=VectorParameter() {defaultValue=<rgba 0.5 0.3095 0 1>}.RGB) {g=true, r=true}
   Radius=Constant() {r=1.2152}
   Density=Constant() {r=1.0128}
  ) {materialFunction=&RadialGradientExponential}
  n22 = Constant() {r=3.9682}
  LinearInterpolate(
   A=@n18
   B=LinearInterpolate(
    A=@n18
    B=VectorParameter() {parameterName=Gradient-Col3, defaultValue=<rgba 0.8148 0.4625 0.0598 1>}.RGB
    Alpha=Power(Base=@n21.RadialGradientExponential, Exp=@n22)
   )
   Alpha=ScalarParameter() {parameterName=IsLockDown}
  )
  Rotator(Time=Constant() {r=-0.025}) {centerY=0.25, speed=6.2932}
 ]
}
