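# Translucent UI master material for item details headers, tinted by rarity colour parameters.
# bIsSchematic blends a radial rarity gradient toward a procedural blueprint look built from sine-based grid lines and wave bands.
# sRarityBorder blends that fill toward a rotated-gradient stroke, scrolled over time via bAnimateStroke, with opacity masked by box distance and scaled by sOpacity.
# A bIsHero branch of sphere-gradient shapes feeds no other node shown.
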Material /FortUILibrary/Styles/Border/Materials/M-Border-Item-DetailsHeader-Master
graph {
 properties {
  bUsedWithStaticMesh=false
  blendMode=BLEND_Translucent
  materialDomain=MD_UI
  pixelDepthOffsetMode=PDOM_Legacy
  usedShadingModels=Default Lit
 }
 nodes [
  n1 = MaterialFunctionCall() {materialFunction=&GetUserInterfaceUV}
  n2 = ScalarParameter() {defaultValue=0.625, parameterName=gGradientAngle}
  n3 = Time()
  n4 = ScalarParameter() {defaultValue=0.35, parameterName=gTime}
  n5 = Frac(None=Add(A=Rotator(Coordinate=@n1.Normalized UV, Time=@n2) {speed=6.2832}, B=Multiply(A=@n3, B=@n4)))
  n6 = ComponentMask(None=Abs(None=Add(A=@n5) {constB=-0.5})) {r=true}
  n7 = ScalarParameter() {parameterName=gGradientMin}
  n8 = ScalarParameter() {defaultValue=1, parameterName=gGradientMax}
  n9 = MaterialFunctionCall(Alpha=@n6, Min=@n7, Max=@n8) {materialFunction=&SmoothStep}
  DotProduct(A=@n6, B=@n9.Result)
  n10 = VectorParameter() {group=Vector Parameter Values, parameterName=Rarity 3, defaultValue=<rgba 0 0.2159 0.0015 1>}
  n11 = VectorParameter() {group=Vector Parameter Values, parameterName=Rarity 2, defaultValue=<rgba 0.1195 0.521 0 1>}
  n12 = Add(B=Constant2Vector() {g=0.25, r=-0.17})
  n13 = ComponentMask(None=Add(A=@n12, B=Constant2Vector())) {r=true}
  n14 = VectorParameter() {group=Vector Parameter Values, parameterName=Rarity 4, defaultValue=<rgba 0.0006 0.0782 0.0009 1>}
  n15 = ComponentMask(None=Add(A=@n12, B=Constant2Vector())) {r=true}
  n16 = Constant() {r=0.5}
  n17 = Constant() {r=0.502}
  n18 = MaterialFunctionCall(
   Alpha=MaterialFunctionCall!(UVs=@n12, CenterPosition=Constant2Vector() {g=1.5, r=-1.6}, Radius=ScalarParameter() {defaultValue=3, parameterName=gInner Circle_c}) {materialFunction=/Engine/Functions/Engine_MaterialFunctions01/Gradient/SphereGradient-2D}.Result 0-1
   Min=@n16
   Max=@n17
  ) {materialFunction=&SmoothStep}
  n19 = Constant2Vector() {g=1, r=0.5}
  n20 = MaterialFunctionCall(
   Alpha=MaterialFunctionCall(
    UVs=Multiply(A=@n12, B=@n19)
    CenterPosition=Constant2Vector() {g=2.5, r=1.75}
    Radius=ScalarParameter() {defaultValue=3, parameterName=gInner Circle_b}
   ) {materialFunction=&SphereGradient-2D}.Result 0-1
   Min=@n16
   Max=@n17
  ) {materialFunction=&SmoothStep}
  n21 = LinearInterpolate(Alpha=@n18.Result)
  n22 = MaterialFunctionCall(
   UVs=Multiply(A=@n12, B=Constant2Vector() {g=1, r=0.45})
   CenterPosition=Constant2Vector() {g=-1.6, r=-0.85}
   Radius=ScalarParameter() {defaultValue=3, parameterName=gInner Circle_a}
  ) {materialFunction=&SphereGradient-2D}
  n23 = Add(
   A=LinearInterpolate(A=LinearInterpolate(Alpha=@n18.Result) {constA=1, constB=0}, Alpha=@n20.Result) {constB=0}
   B=LinearInterpolate(
    B=LinearInterpolate(B=@n21, Alpha=@n20.Result)
    Alpha=MaterialFunctionCall(Alpha=@n22.Result 0-1, Min=@n16, Max=@n17) {materialFunction=&SmoothStep}.Result
   )
  )
  LinearInterpolate(
   A=LinearInterpolate(A=LinearInterpolate(A=@n10.RGB, B=@n11.RGB, Alpha=@n13), B=LinearInterpolate(A=@n14.RGB, B=@n10.RGB, Alpha=@n15), Alpha=Clamp(None=@n23))
   Alpha=ScalarParameter() {group=ItemType, parameterName=bIsHero, sliderMax=1}
  )
  n24 = MaterialFunctionCall() {materialFunction=&GetUserInterfaceUV}
  n25 = MaterialFunctionCall(Float2=Max(A=@n24.Pixel Size)) {materialFunction=&BreakOutFloat2Components}
  n26 = AppendVector(A=Constant() {r=1}, B=Divide(B=Divide(A=@n25.R, B=@n25.R)))
  n27 = Multiply(A=@n24.Normalized UV, B=@n26)
  n28 = MaterialFunctionCall(
   UVs=@n27
   CenterPosition=@n26
   Radius=ScalarParameter() {defaultValue=1, parameterName=sGradientRadius}
   Density=ScalarParameter() {defaultValue=2, parameterName=sGradientDensity}
  ) {materialFunction=&RadialGradientExponential}
  n29 = VectorParameter() {group=Vector Parameter Values, parameterName=Rarity 3, defaultValue=<rgba 0 0.2159 0.0015 1>}
  n30 = VectorParameter() {group=Vector Parameter Values, parameterName=Rarity 2, defaultValue=<rgba 0.1195 0.521 0 1>}
  n31 = VectorParameter() {group=Grid, parameterName=sSchematic-Light-BG, defaultValue=<rgba 0.85 0.85 0.85 1>}
  n32 = VectorParameter() {group=Vector Parameter Values, parameterName=Rarity 2, defaultValue=<rgba 0.1195 0.521 0 1>}
  n33 = Add(A=@n27, B=Constant2Vector()) {constB=0}
  n34 = Multiply(A=@n33, B=Constant2Vector() {g=4, r=2})
  n35 = Constant2Vector() {g=-0.025}
  n36 = Add(A=ComponentMask(None=Cosine(None=@n34) {period=0.5}) {r=true}, B=Multiply(A=Add(A=@n35, B=@n33), B=Constant2Vector() {g=16, r=1}) {constB=2})
  n37 = Constant2Vector()
  n38 = Add(A=Multiply(A=Add(A=@n33, B=@n37), B=Constant2Vector() {g=36, r=1}) {constB=2}, B=ComponentMask(None=Sine(None=@n34) {period=0.4}) {r=true})
  n39 = Clamp(None=Add(A=ComponentMask(None=@n36) {g=true}, B=ComponentMask(None=@n38) {g=true}))
  n40 = MaterialFunctionCall(Alpha=@n39, Min=Min(A=@n39) {constB=0.75}, Max=Max(A=@n39) {constB=0.85}) {materialFunction=&SmoothStep}
  n41 = Sine(None=@n33) {period=2}
  n42 = Multiply(A=ComponentMask(None=@n41) {r=true}, B=ComponentMask(None=Power(Base=@n41) {constExponent=0.8}) {g=true})
  n43 = LinearInterpolate(A=@n31.RGB, B=@n32.RGB, Alpha=Multiply(A=Multiply(A=@n40.Result) {constB=0.75}, B=@n42))
  n44 = Power(Base=@n42) {constExponent=1}
  n45 = LinearInterpolate(
   A=Multiply(A=Multiply(A=@n32.RGB)) {constB=0.75}
   Alpha=LinearInterpolate(
    A=@n31.RGB
    B=VectorParameter() {group=Grid, parameterName=sSchematic-Dark-BG, defaultValue=<rgba 0.55 0.55 0.55 1>}.RGB
    Alpha=Clamp(None=@n44)
   )
  ) {constB=0.6}
  n46 = MaterialFunctionCall() {materialFunction=&GetUserInterfaceUV}
  n47 = MaterialFunctionCall(Float2=Max(A=@n46.Pixel Size)) {materialFunction=&BreakOutFloat2Components}
  n48 = Multiply(A=@n46.Normalized UV, B=AppendVector(A=Divide(A=@n47.R, B=@n47.R), B=Constant() {r=1}))
  n49 = Sine(None=ComponentMask(None=@n48) {r=true}) {period=0.125}
  n50 = ScalarParameter() {defaultValue=0.9, group=Grid, parameterName=sMin}
  n51 = ScalarParameter() {defaultValue=1, group=Grid, parameterName=sMax}
  n52 = MaterialFunctionCall(Alpha=@n49, Min=Min(A=@n49, B=@n50), Max=Max(A=@n49, B=@n51)) {materialFunction=&SmoothStep}
  n53 = Sine(None=ComponentMask(None=@n48) {g=true}) {period=0.125}
  n54 = MaterialFunctionCall(Alpha=@n53, Min=Min(A=@n53, B=@n50), Max=Max(A=@n53, B=@n51)) {materialFunction=&SmoothStep}
  n55 = LinearInterpolate(B=Clamp(None=LinearInterpolate(A=@n52.Result, Alpha=@n54.Result)), Alpha=Multiply(A=@n42) {constB=1.5})
  n56 = VectorParameter() {group=Vector Parameter Values, parameterName=Rarity 2, defaultValue=<rgba 0.1195 0.521 0 1>}
  n57 = ScalarParameter() {group=Border, parameterName=bAnimateStroke}
  n58 = ComponentMask(None=Rotator(Coordinate=@n1.Normalized UV, Time=@n2) {speed=6.2832}) {r=true}
  n59 = Clamp(None=DotProduct(
    A=Add(A=@n58) {constB=0}
    B=MaterialFunctionCall(Alpha=@n58, Min=@n7, Max=@n8) {materialFunction=&SmoothStep}.Result
   ))
  n60 = Clamp(None=@n9.Result)
  n61 = ScalarParameter() {group=ItemType, parameterName=sRarityBorder, sliderMax=1}
  LinearInterpolate(
   A=LinearInterpolate(
    A=LinearInterpolate(A=@n29.RGB, B=@n30.RGB, Alpha=@n28.RadialGradientExponential)
    B=LinearInterpolate(A=@n43, B=@n45, Alpha=@n55)
    Alpha=ScalarParameter() {defaultValue=1, group=ItemType, parameterName=bIsSchematic, sliderMax=1}
   )
   B=LinearInterpolate(
    A=VectorParameter() {group=Vector Parameter Values, parameterName=Rarity 4, defaultValue=<rgba 0.0006 0.0782 0.0009 1>}.RGB
    B=Add(A=@n56.RGB) {constB=0.25}
    Alpha=LinearInterpolate(A=@n59, B=@n60, Alpha=@n57)
   )
   Alpha=@n61
  )
  n62 = Multiply(A=TextureCoordinate(), B=Constant2Vector() {g=1, r=1})
  n63 = Subtract(A=Abs(None=Add(A=@n62) {constB=-0.5}), B=ScalarParameter() {defaultValue=0.45, group=Border, parameterName=gStrokeThickness, sliderMax=1})
  n64 = MaterialFunctionCall(
   Alpha=MaterialFunctionCall(Vector 2=Max(A=@n63) {constB=0}) {materialFunction=/Engine/Functions/Engine_MaterialFunctions02/Utility/VectorLength}.V2 Length
   Min=ScalarParameter() {group=Border, parameterName=gStrokeMin}
   Max=ScalarParameter() {defaultValue=0.001, group=Border, parameterName=gStrokeMax}
  ) {materialFunction=&SmoothStep}
  Multiply(
   A=LinearInterpolate(A=Constant() {r=1}, B=Multiply(A=@n64.Result), Alpha=@n61)
   B=ScalarParameter() {defaultValue=1, parameterName=sOpacity, sliderMax=1}
  )
  n65 = MaterialFunctionCall() {materialFunction=&BreakOutFloat2Components}
  n66 = Divide(A=@n65.R, B=@n65.R)
  Multiply(B=AppendVector(A=Constant() {r=1}, B=Divide(B=@n66)))
  n67 = MaterialFunctionCall() {materialFunction=&BreakOutFloat2Components}
  n68 = Divide(A=@n67.R, B=@n67.R)
  n69 = Divide(A=@n67.R, B=@n67.R)
  Multiply(
   A=MaterialFunctionCall() {materialFunction=&GetUserInterfaceUV}.Normalized UV
   B=AppendVector(A=@n68, B=Divide(A=@n68, B=@n69))
  )
  n70 = MaterialFunctionCall() {materialFunction=&GetUserInterfaceUV}
  n71 = MaterialFunctionCall(Float2=@n70.Pixel Size) {materialFunction=&BreakOutFloat2Components}
  n72 = Divide(A=@n71.R, B=@n71.R)
  n73 = Divide(A=@n71.R, B=@n71.R)
  Multiply(A=@n70.Normalized UV, B=AppendVector(A=@n72, B=Divide(A=@n72, B=@n73)))
  TextureCoordinate()
  TextureCoordinate()
  VectorParameter() {group=Vector Parameter Values, parameterName=Rarity 4, defaultValue=<rgba 0.0006 0.0782 0.0009 1>}
 ]
}
