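# Translucent UI material that draws the rarity-colored background and border of a battle loot tile.
# Offset radial gradients and rotated-UV glows blend and shade the rarity colors, with optional desaturation.
# A distance-based edge mask confines an angled highlight and an optional time-scrolled shine added in the first rarity color.
# BorderOnly blends from full coverage to that edge mask, and separate InnerAngles parameters build angled top and bottom masks.
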Material /FortUILibrary/Styles/Border/Materials/M_Border_BattleLootTile_ColorBG
graph {
 properties {
  bUsedWithStaticMesh=false
  blendMode=BLEND_Translucent
  materialDomain=MD_UI
  pixelDepthOffsetMode=PDOM_Legacy
  usedShadingModels=Default Lit
 }
 nodes [
  n1 = VectorParameter() {group=RarityColors, parameterName=RGB_3, defaultValue=<rgba 0.1653 0.171 0.29 1>}
  n2 = VectorParameter() {group=RarityColors, parameterName=RGB_2, defaultValue=<rgba 0.5538 0.585 0.78 1>}
  n3 = MaterialFunctionCall() {materialFunction=&GetUserInterfaceUV}
  n4 = MaterialFunctionCall(
   UVs=@n3.Normalized UV
   CenterPosition=AppendVector(
    A=ScalarParameter() {defaultValue=0.6, parameterName=AGradientOffsetX}
    B=ScalarParameter() {defaultValue=0.4, parameterName=AGradientOffsetY}
   )
   Radius=Constant() {r=0.75}
   Density=Constant() {r=1.5}
  ) {materialFunction=&RadialGradientExponential}
  n5 = Rotator(Coordinate=@n3.Normalized UV, Time=Divide(A=ScalarParameter() {defaultValue=-60, parameterName=AngledGlowLeft}) {constA=360, constB=360}) {speed=6.2832}
  n6 = Rotator(Coordinate=@n3.Normalized UV, Time=Divide(A=ScalarParameter() {defaultValue=30, parameterName=AngledGlowRight}) {constA=360, constB=360}) {speed=6.2832}
  n7 = MaterialFunctionCall(Alpha=DotProduct(A=ComponentMask(None=@n5) {g=true}, B=ComponentMask(None=@n6) {g=true}), Min=Constant(), Max=Constant() {r=1}) {materialFunction=&SmoothStep}
  n8 = ScalarParameter() {defaultValue=0.6, parameterName=BGradientOffsetX}
  n9 = ScalarParameter() {defaultValue=0.55, parameterName=BGradientOffsetY}
  n10 = Add(
   A=MaterialFunctionCall(UVs=@n3.Normalized UV, CenterPosition=AppendVector(A=@n8, B=@n9), Radius=Constant() {r=0.875}, Density=Constant() {r=1.32}) {materialFunction=&RadialGradientExponential}.RadialGradientExponential
   B=Constant() {r=0.38}
  )
  n11 = Add(A=Multiply(A=TextureCoordinate(), B=Constant2Vector() {g=1, r=1})) {constB=-0.5}
  n12 = Max(A=Subtract(A=Abs(None=@n11), B=Constant() {r=0.4375})) {constB=0}
  n13 = Clamp(None=MaterialFunctionCall(
    Alpha=MaterialFunctionCall(Vector 2=@n12) {materialFunction=/Engine/Functions/Engine_MaterialFunctions02/Utility/VectorLength}.V2 Length
    Min=Constant()
    Max=Constant() {r=0.001}
   ) {materialFunction=&SmoothStep}.Result)
  n14 = Multiply(
   A=LinearInterpolate(
    A=LinearInterpolate(A=@n1.RGB, B=@n2.RGB, Alpha=@n4.RadialGradientExponential)
    B=VectorParameter() {group=RarityColors, parameterName=RGB_4, defaultValue=<rgba 0.0731 0.0689 0.135 1>}.RGB
    Alpha=Multiply(A=@n7.Result)
   )
   B=LinearInterpolate(A=Clamp(None=@n10), Alpha=@n13)
  )
  n15 = Rotator(Coordinate=Subtract(A=@n3.Normalized UV) {constB=0.5}, Time=Divide(A=Constant() {r=-60}) {constA=360, constB=360}) {centerX=0, centerY=0, speed=6.2832}
  n16 = ComponentMask(None=@n15) {g=true}
  n17 = Constant()
  n18 = Clamp(None=OneMinus(None=MaterialFunctionCall(Alpha=@n16, Min=@n17, Max=@n17) {materialFunction=&SmoothStep}.Result)) {minDefault=0.33}
  n19 = MaterialFunctionCall(Alpha=ComponentMask(None=Abs(None=@n15)) {g=true}, Min=@n17, Max=Constant() {r=1}) {materialFunction=&SmoothStep}
  n20 = DotProduct(A=LinearInterpolate(Alpha=Multiply(A=@n18, B=@n19.Result)) {constA=0.25}, B=@n13)
  n21 = Divide(A=Constant() {r=30}) {constA=360, constB=360}
  n22 = Time()
  n23 = ScalarParameter() {defaultValue=-1, parameterName=s_Time}
  n24 = Frac(None=Add(A=Rotator(Coordinate=@n3.Normalized UV, Time=@n21) {centerX=0, centerY=0, speed=6.2832}, B=Multiply(A=@n22, B=@n23)))
  n25 = ComponentMask(None=Abs(None=Add(A=@n24) {constB=-0.5})) {r=true}
  n26 = Constant()
  n27 = Constant() {r=1}
  n28 = LinearInterpolate(
   B=LinearInterpolate(Alpha=MaterialFunctionCall(Alpha=@n25, Min=@n26, Max=@n27) {materialFunction=&SmoothStep}.Result)
   Alpha=@n13
  )
  n29 = ScalarParameter() {parameterName=s_AnimateBorder}
  Add(
   A=Desaturation(None=LinearInterpolate(A=@n14, B=@n2.RGB, Alpha=@n20), Fraction=ScalarParameter() {parameterName=DesaturationFactor})
   B=LinearInterpolate(
    B=VectorParameter() {group=RarityColors, parameterName=RGB_1, defaultValue=<rgba 1 1 1 1>}.RGB
    Alpha=LinearInterpolate(B=@n28, Alpha=@n29) {constB=0}
   )
  )
  n30 = OneMinus(None=Add(B=ScalarParameter() {defaultValue=0.5, group=InnerAngles, parameterName=s_BottomAngleOffset, sliderMax=1}))
  n31 = Divide(A=ScalarParameter() {defaultValue=-2, group=InnerAngles, parameterName=s_BottomAngle}) {constA=360, constB=360}
  n32 = Constant()
  n33 = Constant() {r=0.005}
  n34 = MaterialFunctionCall(Alpha=ComponentMask(None=Rotator(Coordinate=@n30, Time=@n31) {centerX=0, centerY=0, speed=6.2832}) {g=true}, Min=@n32, Max=@n33) {materialFunction=&SmoothStep}
  Clamp(None=Multiply(A=OneMinus(None=@n34.Result), B=ScalarParameter() {defaultValue=0.3, group=InnerAngles, parameterName=s_UseBottomAngle, sliderMax=1}))
  n35 = ScalarParameter() {defaultValue=0.125, group=InnerAngles, parameterName=s_TopAngleOffset, sliderMax=1}
  n36 = ScalarParameter() {defaultValue=-4, group=InnerAngles, parameterName=s_TopAngle}
  n37 = ComponentMask(None=Rotator(Coordinate=Subtract(B=@n35) {constB=0.125}, Time=Divide(A=@n36) {constA=360, constB=360}) {centerX=0, centerY=0, speed=6.2832}) {g=true}
  Clamp(None=Add(
    A=MaterialFunctionCall(Alpha=@n37, Min=@n32, Max=@n33) {materialFunction=&SmoothStep}.Result
    B=ScalarParameter() {defaultValue=0.95, group=InnerAngles, parameterName=s_UseTopAngle, sliderMax=1}
   ))
  LinearInterpolate(B=@n13, Alpha=ScalarParameter() {parameterName=BorderOnly, sliderMax=1}) {constA=1}
 ]
}
