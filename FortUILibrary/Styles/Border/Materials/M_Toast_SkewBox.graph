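# Translucent UI material for a skewed toast box that shows a small preview image, by default a hero portrait, over a gradient background.
# BorderTexture channels are smoothstepped into masks: R clips the preview, G shapes overall opacity, B separates edge from interior.
# Vertical UV gradients drive edge color, inner color and edge alpha, with the edge color added over the inner gradient.
# The scaled, offset preview composites by its alpha, and RGBOverrideSwitch blends in an additive OverrideRGB tint.
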Material /FortUILibrary/Styles/Border/Materials/M_Toast_SkewBox
graph {
 properties {
  bUsedWithStaticMesh=false
  blendMode=BLEND_Translucent
  materialDomain=MD_UI
  pixelDepthOffsetMode=PDOM_Legacy
  usedShadingModels=Default Lit
 }
 nodes [
  ComponentMask() {g=true, r=true}
  n1 = ComponentMask(None=TextureCoordinate()) {g=true}
  n2 = LinearInterpolate(
   B=VectorParameter() {parameterName=EdgeRGB, defaultValue=<rgba 1 1 1 1>}.RGB
   Alpha=LinearInterpolate(A=ScalarParameter() {defaultValue=0.7, parameterName=EdgeColorFade}, Alpha=@n1)
  )
  n3 = TextureSampleParameter2D() {parameterName=BorderTexture, samplerType=SAMPLERTYPE_LinearColor, texture=/Game/UI/Foundation/Textures/Border/T_Toast_SkewBoxPortrait}
  n4 = Constant()
  n5 = Constant() {r=0.125}
  n6 = MaterialFunctionCall(Alpha=@n3.B, Min=@n4, Max=@n5) {materialFunction=&SmoothStep}
  n7 = VectorParameter() {parameterName=InnerFadeRGB, defaultValue=<rgba 0.051 0.0122 1 1>}
  n8 = VectorParameter() {parameterName=InnerRGB, defaultValue=<rgba 0.0935 0 0.2604 1>}
  n9 = LinearInterpolate(A=ScalarParameter() {defaultValue=0.7, parameterName=InnerColorFade}, Alpha=@n1)
  n10 = Constant() {r=0.925}
  n11 = TextureSampleParameter2D(UVs=Add(
    A=MaterialFunctionCall(Texture Scale=@n10) {materialFunction=&ScaleUVsByCenter}.UVs
    B=Constant2Vector() {g=0.025}
   )) {
   parameterName=SmallPreviewImage
   samplerSource=SSM_Clamp_WorldGroupSettings
   texture=/Game/UI/Foundation/Textures/Icons/Heroes/Athena/Soldier/T-Soldier-HID-005-Athena-Commando-M
  }
  n12 = LinearInterpolate(
   B=@n11.A
   Alpha=MaterialFunctionCall(Alpha=@n3.R, Min=@n4, Max=@n5) {materialFunction=&SmoothStep}.Result
  )
  n13 = LinearInterpolate(A=Add(A=LinearInterpolate(A=@n2, Alpha=@n6.Result) {constB=0}, B=LinearInterpolate(A=@n7.RGB, B=@n8.RGB, Alpha=@n9)), B=@n11.RGB, Alpha=@n12)
  LinearInterpolate(
   A=@n13
   B=Add(A=VectorParameter() {parameterName=OverrideRGB, defaultValue=<rgba 1 1 1 1>}.RGB, B=@n13)
   Alpha=ScalarParameter() {parameterName=RGBOverrideSwitch}
  )
  n14 = LinearInterpolate(
   A=ScalarParameter() {defaultValue=0.5, parameterName=EdgeAlphaFadeA}
   B=ScalarParameter() {defaultValue=1, parameterName=EdgeAlphaFadeB}
   Alpha=@n1
  )
  n15 = ScalarParameter() {defaultValue=0.5, parameterName=InnerAlpha}
  LinearInterpolate(
   A=LinearInterpolate(
    B=LinearInterpolate(A=@n14, B=@n15, Alpha=@n6.Result)
    Alpha=MaterialFunctionCall(Alpha=@n3.G, Min=@n4, Max=@n5) {materialFunction=&SmoothStep}.Result
   )
   Alpha=@n12
  )
 ]
}
