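# Translucent UI border-style material that draws an animated, color-parameterized backdrop of scrolling cube and glyph patterns.
# Tiled, rotated cube samples, shaped by radial gradients, blend base colors, while noise-modulated glyph channels add edge glow and brightness.
# A second glyph layer, masked by a sine-animated cube pattern, is added on top; several crossfades blend a channel with itself.
# The UseAlpha static switch selects summed-mask opacity or constant one.
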Material /FortUILibrary/Styles/Border/Materials/M_UI_CubePattern
graph {
 properties {
  bUsedWithStaticMesh=false
  blendMode=BLEND_Translucent
  materialDomain=MD_UI
  pixelDepthOffsetMode=PDOM_Legacy
  usedShadingModels=Default Lit
 }
 nodes [
  n1 = VectorParameter() {parameterName=RGB2, defaultValue=<rgba 0.2328 0 0.3021 1>}
  n2 = Multiply(
   A=TextureCoordinate()
   B=AppendVector(A=ScalarParameter() {defaultValue=2, parameterName=TileU}, B=ScalarParameter() {defaultValue=1, parameterName=TileV})
  ) {constB=3}
  n3 = Rotator(Coordinate=@n2, Time=Constant() {r=0.125})
  n4 = TextureSample(UVs=Panner(Coordinate=Multiply(A=@n3) {constB=0.625}) {speedX=-0.01, speedY=0.03}) {samplerType=SAMPLERTYPE_LinearColor, texture=/Game/UI/Foundation/Textures/Pattern/T_CubePattern}
  n5 = Time()
  n6 = Multiply(A=@n5, B=ScalarParameter() {defaultValue=0.25, parameterName=SquareTimeMult}) {constB=0.2}
  n7 = Multiply(A=Add(A=Sine(None=@n6))) {constB=0.5}
  n8 = LinearInterpolate(A=ScalarParameter() {parameterName=CutoutSquareAlpha}, Alpha=LinearInterpolate(A=@n4.R, B=@n4.R, Alpha=@n7))
  n9 = TextureSample(UVs=Panner(Coordinate=@n3, Time=@n5) {speedX=0.01, speedY=0.025}) {samplerType=SAMPLERTYPE_LinearColor, texture=&T_CubePattern}
  n10 = Multiply(A=@n6) {constB=2}
  n11 = Multiply(A=Add(A=Sine(None=@n10))) {constB=0.5}
  n12 = LinearInterpolate(
   B=LinearInterpolate(A=LinearInterpolate(A=@n9.R, B=@n9.R, Alpha=@n11), B=LinearInterpolate(A=@n9.B, B=@n9.B, Alpha=@n11), Alpha=@n7)
   Alpha=MaterialFunctionCall(
    Radius=ScalarParameter() {defaultValue=0.625, parameterName=SquareRadius}
    Density=ScalarParameter() {defaultValue=1.5, parameterName=SquareDensity}
   ) {materialFunction=&RadialGradientExponential}.RadialGradientExponential
  )
  n13 = Multiply(
   A=MaterialFunctionCall(
    Radius=ScalarParameter() {defaultValue=0.45, parameterName=EdgingRadius}
    Density=ScalarParameter() {defaultValue=1.25, parameterName=EdgingDensity}
   ) {materialFunction=&RadialGradientExponential}.RadialGradientExponential
   B=ScalarParameter() {defaultValue=3, parameterName=EdgingMultiplyPower}
  ) {constB=3}
  n14 = LinearInterpolate(
   A=LinearInterpolate(A=VectorParameter() {parameterName=RGB1, defaultValue=<rgba 0.0112 0 0.0573 1>}.RGB, B=@n1.RGB, Alpha=Multiply(A=@n8, B=@n12))
   B=@n1.RGB
   Alpha=Multiply(A=@n13, B=ScalarParameter() {parameterName=RGB3HotSpotAlpha}) {constB=0.125}
  )
  n15 = VectorParameter() {parameterName=RGB3, defaultValue=<rgba 1 0 0.5856 1>}
  n16 = TextureSample(UVs=Panner(Coordinate=@n3, Time=@n5) {speedX=-0.01, speedY=0.03}) {samplerType=SAMPLERTYPE_LinearColor, texture=/Game/UI/Foundation/Textures/Pattern/T_GlyphPattern}
  n17 = TextureSample(UVs=Panner(Coordinate=@n2, Time=@n5) {speedX=0.175, speedY=0.075}) {samplerType=SAMPLERTYPE_Grayscale, texture=&LowResBlurredNoise}
  n18 = Clamp(None=Power(Base=Multiply(A=@n17.RGB) {constB=2}))
  n19 = Clamp(None=@n13, Min=ScalarParameter() {parameterName=EdgingBaseAlpha}) {maxDefault=2, minDefault=0.05}
  n20 = Multiply(A=Multiply(A=Multiply(A=@n16.G, B=@n18), B=@n19), B=ScalarParameter() {defaultValue=0.5, parameterName=EdgingGlow}) {constB=0.5}
  n21 = Multiply(A=Multiply(A=Multiply(A=@n16.R, B=@n18), B=@n19), B=ScalarParameter() {defaultValue=0.5, parameterName=EdgingBrightness}) {constB=0.5}
  n22 = VectorParameter() {parameterName=RGB5, defaultValue=<rgba 0.929 0 1 1>}
  n23 = VectorParameter() {parameterName=RGB6, defaultValue=<rgba 0.9667 0.5312 1 1>}
  n24 = TextureCoordinate()
  n25 = AppendVector(A=ScalarParameter() {defaultValue=4, parameterName=TileUGlyph}, B=ScalarParameter() {defaultValue=2, parameterName=TileVGlyph})
  n26 = Time()
  n27 = ScalarParameter() {defaultValue=-0.01, parameterName=PannerSpeedx}
  n28 = Multiply(A=@n27) {constB=4}
  n29 = ScalarParameter() {defaultValue=0.025, parameterName=PannerSpeedY}
  n30 = Multiply(A=@n29) {constB=4}
  n31 = TextureSample(UVs=Panner(Coordinate=Multiply(A=@n24, B=@n25) {constB=3}, Time=@n26, Speed=AppendVector(A=@n28, B=@n30)) {speedX=-1}) {samplerType=SAMPLERTYPE_LinearColor, texture=&T_GlyphPattern}
  n32 = MaterialFunctionCall(Alpha=@n31.B, Min=Constant() {r=0.5}) {materialFunction=&SmoothStep}
  n33 = TextureCoordinate() {uTiling=0.25, vTiling=0.25}
  n34 = TextureSample(UVs=Panner(Coordinate=Multiply(A=@n33, B=@n25), Time=@n26, Speed=AppendVector(A=@n27, B=@n29)) {speedX=-0.25}) {samplerType=SAMPLERTYPE_LinearColor, texture=&T_CubePattern}
  n35 = Multiply(A=@n26, B=ScalarParameter() {defaultValue=0.25, parameterName=GlyphTimeMult}) {constB=0.2}
  n36 = Multiply(A=@n35) {constB=2}
  n37 = Multiply(A=Add(A=Sine(None=@n36))) {constB=0.5}
  n38 = Add(A=Sine(None=@n35))
  n39 = Multiply(
   A=LinearInterpolate(
    A=LinearInterpolate(A=@n34.R, B=@n34.R, Alpha=@n37)
    B=LinearInterpolate(A=@n34.B, B=@n34.B, Alpha=@n37)
    Alpha=Multiply(A=@n38) {constB=0.5}
   )
   B=ScalarParameter() {defaultValue=8, parameterName=GLpyhMultFactor}
  ) {constB=4}
  n40 = LinearInterpolate(B=Add(A=Power(Base=@n31.A), B=@n32.Result), Alpha=Clamp(None=OneMinus(None=@n39)))
  Add(
   A=LinearInterpolate(
    A=LinearInterpolate(A=@n14, B=@n15.RGB, Alpha=@n20)
    B=VectorParameter() {parameterName=RGB4, defaultValue=<rgba 1 0.1979 0.8089 1>}.RGB
    Alpha=@n21
   )
   B=LinearInterpolate(B=LinearInterpolate(A=@n22.RGB, B=@n23.RGB, Alpha=@n32.Result), Alpha=@n40)
  )
  n41 = Add(A=@n21, B=@n12)
  StaticSwitchParameter(True=Add(A=@n19, B=Add(A=@n41, B=@n40)), False=Constant() {r=1}) {parameterName=UseAlpha}
 ]
}
