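# Translucent UI material that draws a check mark icon inside a filled circle with a contrasting ring.
# An exponential radial gradient through two SmoothSteps forms the outer edge and inner fill, blending parameter colors and alphas.
# The ring color mixes base and accent by ColorFillSplit.
# A scaled, offset icon texture's alpha lays the check over the fill, with a static switch choosing accent or white, and raises opacity.
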Material /FortUILibrary/Styles/Border/Materials/M_CircledCheckMark
graph {
 properties {
  bUsedWithStaticMesh=false
  blendMode=BLEND_Translucent
  materialDomain=MD_UI
  pixelDepthOffsetMode=PDOM_Legacy
  usedShadingModels=Default Lit
 }
 nodes [
  n1 = VectorParameter() {parameterName=RGBA_3, defaultValue=<rgba 0.0343 1 0.227 0.8>}
  n2 = VectorParameter() {parameterName=RGBA_Base, defaultValue=<rgba 0 0.0021 0.0176 0.8>}
  n3 = MaterialFunctionCall(UVs=TextureCoordinate() {coordinateIndex=4}) {materialFunction=&RadialGradientExponential}
  n4 = MaterialFunctionCall(Alpha=@n3.RadialGradientExponential, Min=Constant() {r=0.15}, Max=Constant() {r=0.25}) {materialFunction=&SmoothStep}
  n5 = Constant()
  n6 = Constant() {r=0.05}
  n7 = ScalarParameter() {defaultValue=-0.25, parameterName=IconUOffset}
  n8 = ScalarParameter() {defaultValue=-0.325, parameterName=IconVOffset}
  n9 = TextureSampleParameter2D(UVs=Add(A=TextureCoordinate() {uTiling=1.625, vTiling=1.625}, B=AppendVector(A=@n7, B=@n8)) {constB=-0.25}) {parameterName=IconTexture, texture=/Game/UI/Foundation/Textures/Icons/Checks/T-Icon-Check-32}
  LinearInterpolate(
   A=LinearInterpolate(
    B=LinearInterpolate(A=@n1.A, B=@n2.A, Alpha=@n4.Result)
    Alpha=MaterialFunctionCall(Alpha=@n3.RadialGradientExponential, Min=@n5, Max=@n6) {materialFunction=&SmoothStep}.Result
   )
   Alpha=@n9.A
  )
  n10 = Constant() {r=1}
  LinearInterpolate(
   A=LinearInterpolate(A=@n2.RGB, B=@n1.RGB, Alpha=ScalarParameter() {defaultValue=0.5, parameterName=ColorFillSplit})
   B=LinearInterpolate(A=@n2.RGB, B=StaticSwitchParameter(True=@n1.RGB, False=@n10) {defaultValue=true, parameterName=ColorIcon}, Alpha=@n9.A)
   Alpha=@n4.Result
  )
 ]
}
